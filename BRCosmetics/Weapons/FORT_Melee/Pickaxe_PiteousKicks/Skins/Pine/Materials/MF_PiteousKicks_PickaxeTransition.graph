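# Material function for the PiteousKicks Pine pickaxe that animates a transition to a gold material as the SL_CurrentLevel parameter changes.
# A mask combines world-aligned noise on pre-skinned position and normal with a height threshold, which moves between per-level MaskPos values over TransitionDuration from BP_GetTime.
# The mask blends incoming attributes with gold textures, and a thresholded edge band adds emissive glow that fades as the timer ends.
# Some roughness and specular adjustment nodes do not reach the output.
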MaterialFunction /BRCosmetics/Weapons/FORT_Melee/Pickaxe_PiteousKicks/Skins/Pine/Materials/MF_PiteousKicks_PickaxeTransition
graph {expressions {
  MaterialExpressionAdd_0 {
   class=MaterialExpressionAdd
   inputs [
    A {from=MaterialExpressionNamedRerouteUsage_4}
    B {from=MaterialExpressionScalarParameter_1}
   ]
  }
  MaterialExpressionAdd_1 {
   class=MaterialExpressionAdd
   inputs [
    A {from=MaterialExpressionScalarParameter_12}
    B {from=""}
   ]
   properties {constB=0.135}
  }
  MaterialExpressionAdd_2 {
   class=MaterialExpressionAdd
   inputs [
    A {from=MaterialExpressionSubtract_1}
    B {from=MaterialExpressionLinearInterpolate_3}
   ]
   properties {constB=0}
  }
  MaterialExpressionAdd_3 {
   class=MaterialExpressionAdd
   inputs [
    A {from=MaterialExpressionNamedRerouteUsage_5}
    B {from=MaterialExpressionMultiply_2}
   ]
  }
  MaterialExpressionAdd_5 {
   class=MaterialExpressionAdd
   inputs [
    A {from=MaterialExpressionScalarParameter_3}
    B {from=MaterialExpressionScalarParameter_11}
   ]
  }
  MaterialExpressionBlendMaterialAttributes_0 {
   class=MaterialExpressionBlendMaterialAttributes
   inputs [
    A {from=MaterialExpressionReroute_7, output=""}
    B {from=MaterialExpressionSetMaterialAttributes_0, output=""}
    Alpha {from=MaterialExpressionNamedRerouteUsage_0, output=""}
   ]
   properties {}
  }
  MaterialExpressionComponentMask_2 {class=MaterialExpressionComponentMask, inputs [None {from=MaterialExpressionStaticSwitchParameter_0}], properties {r=true}}
  MaterialExpressionComponentMask_3 {class=MaterialExpressionComponentMask, inputs [None {from=MaterialExpressionStaticSwitchParameter_0}], properties {g=true}}
  MaterialExpressionComponentMask_4 {class=MaterialExpressionComponentMask, inputs [None {from=MaterialExpressionStaticSwitchParameter_0}], properties {b=true}}
  MaterialExpressionComponentMask_5 {
   class=MaterialExpressionComponentMask
   inputs [None {from=MaterialExpressionMaterialFunctionCall_4, output=XYZ Texture}]
   properties {b=true}
  }
  MaterialExpressionComponentMask_6 {class=MaterialExpressionComponentMask, inputs [None {from=MaterialExpressionShaderStageSwitch_3}], properties {b=true}}
  MaterialExpressionConstant_0 {class=MaterialExpressionConstant}
  MaterialExpressionConstant_1 {class=MaterialExpressionConstant, properties {r=1}}
  MaterialExpressionConstant_2 {class=MaterialExpressionConstant, properties {r=1}}
  MaterialExpressionConstant_3 {class=MaterialExpressionConstant, properties {r=1}}
  MaterialExpressionConstant_4 {class=MaterialExpressionConstant}
  MaterialExpressionConstant_5 {class=MaterialExpressionConstant, properties {r=1}}
  MaterialExpressionConstant_7 {class=MaterialExpressionConstant, properties {r=1}}
  MaterialExpressionDivide_0 {
   class=MaterialExpressionDivide
   inputs [
    A {from=MaterialExpressionSubtract_0}
    B {from=MaterialExpressionAdd_1}
   ]
  }
  MaterialExpressionDivide_1 {
   class=MaterialExpressionDivide
   inputs [
    A {from=MaterialExpressionSubtract_3}
    B {from=MaterialExpressionScalarParameter_11}
   ]
  }
  MaterialExpressionFunctionInput_1 {
   class=MaterialExpressionFunctionInput
   inputs [Preview {from="", output=""}]
   properties {inputName=MaterialAttributes, inputType=FunctionInput_MaterialAttributes, sortPriority=10}
  }
  MaterialExpressionFunctionInput_2 {class=MaterialExpressionFunctionInput, inputs [Preview {from="", output=""}], properties {inputName=OriginalNormal, sortPriority=1000}}
  MaterialExpressionFunctionOutput_1 {class=MaterialExpressionFunctionOutput, inputs [None {from=MaterialExpressionSetMaterialAttributes_1, output=""}]}
  MaterialExpressionGetMaterialAttributes_0 {
   class=MaterialExpressionGetMaterialAttributes
   inputs [None {from="", output=""}]
   properties {attributeGetTypes [#1, #2, #3]}
  }
  MaterialExpressionGetMaterialAttributes_1 {
   class=MaterialExpressionGetMaterialAttributes
   inputs [None {from=MaterialExpressionBlendMaterialAttributes_0, output=""}]
   properties {attributeGetTypes [#4]}
  }
  MaterialExpressionIf_0 {
   class=MaterialExpressionIf
   inputs [
    A {from=MaterialExpressionScalarParameter_13, output=""}
    B {from="", output=""}
    "A > B" {from=MaterialExpressionConstant_1, output=""}
    "A == B" {from=MaterialExpressionConstant_0, output=""}
    "A < B" {from=MaterialExpressionConstant_0, output=""}
   ]
   properties {}
  }
  MaterialExpressionIf_2 {
   class=MaterialExpressionIf
   inputs [
    A {from=MaterialExpressionScalarParameter_17, output=""}
    B {from="", output=""}
    "A > B" {from=MaterialExpressionVectorParameter_3, output=RGB}
    "A == B" {from=MaterialExpressionLinearInterpolate_4, output=""}
    "A < B" {from=MaterialExpressionVectorParameter_4, output=RGB}
   ]
   properties {constB=2}
  }
  MaterialExpressionLinearInterpolate_0 {
   class=MaterialExpressionLinearInterpolate
   inputs [
    A {from=MaterialExpressionSaturate_1}
    B {from=MaterialExpressionNamedRerouteUsage_1}
    Alpha {from=MaterialExpressionNamedRerouteUsage_2}
   ]
  }
  MaterialExpressionLinearInterpolate_1 {
   class=MaterialExpressionLinearInterpolate
   inputs [
    A {from=MaterialExpressionSaturate_0}
    B {from=MaterialExpressionNamedRerouteUsage_4}
    Alpha {from=MaterialExpressionNamedRerouteUsage_3}
   ]
  }
  MaterialExpressionLinearInterpolate_2 {
   class=MaterialExpressionLinearInterpolate
   inputs [
    A {from=MaterialExpressionNamedRerouteUsage_5}
    B {from=MaterialExpressionAdd_3}
    Alpha {from=MaterialExpressionNamedRerouteUsage_6}
   ]
  }
  MaterialExpressionLinearInterpolate_3 {
   class=MaterialExpressionLinearInterpolate
   inputs [
    A {from=MaterialExpressionSwitch_0}
    B {from=MaterialExpressionSwitch_1}
    Alpha {from=MaterialExpressionNamedRerouteUsage_9}
   ]
  }
  MaterialExpressionLinearInterpolate_4 {
   class=MaterialExpressionLinearInterpolate
   inputs [
    A {from=MaterialExpressionVectorParameter_3, output=RGB}
    B {from=MaterialExpressionVectorParameter_4, output=RGB}
    Alpha {from=MaterialExpressionNamedRerouteUsage_10}
   ]
  }
  MaterialExpressionMaterialFunctionCall_0 {
   class=MaterialExpressionMaterialFunctionCall
   inputs [
    R {from=MaterialExpressionGetMaterialAttributes_0, output=Specular}
    G {from=MaterialExpressionGetMaterialAttributes_0, output=Specular}
    B {from=MaterialExpressionGetMaterialAttributes_0, output=Specular}
    A {from=""}
   ]
   properties {materialFunction=&AppendMany}
  }
  MaterialExpressionMaterialFunctionCall_4 {
   class=MaterialExpressionMaterialFunctionCall
   inputs [
    Export Float 4 {from=""}
    TextureObject {from=MaterialExpressionTextureObjectParameter_0}
    TextureSize {from=MaterialExpressionIf_2}
    WorldPosition {from=MaterialExpressionShaderStageSwitch_3}
    World Space Normal {from=MaterialExpressionShaderStageSwitch_2}
    RotZAxis {from=""}
    RotXAxis {from=""}
    RotYAxis {from=""}
    ProjectionTransitionContrast {from=""}
   ]
   properties {materialFunction=/Game/Weapons/WeaponSkins/Wraps/MaterialFunctions/WorldAlignedTexture}
  }
  MaterialExpressionMultiply_0 {
   class=MaterialExpressionMultiply
   inputs [
    A {from=MaterialExpressionNamedRerouteUsage_0}
    B {from=MaterialExpressionIf_0}
   ]
  }
  MaterialExpressionMultiply_1 {
   class=MaterialExpressionMultiply
   inputs [
    A {from=MaterialExpressionNamedRerouteUsage_1}
    B {from=MaterialExpressionScalarParameter_0}
   ]
  }
  MaterialExpressionMultiply_2 {
   class=MaterialExpressionMultiply
   inputs [
    A {from=MaterialExpressionVectorParameter_1, output=RGB}
    B {from=MaterialExpressionSmoothStep_0}
   ]
  }
  MaterialExpressionMultiply_4 {
   class=MaterialExpressionMultiply
   inputs [
    A {from=MaterialExpressionComponentMask_6}
    B {from=MaterialExpressionConstant_7}
   ]
   properties {constB=0.005}
  }
  MaterialExpressionMultiply_5 {
   class=MaterialExpressionMultiply
   inputs [
    A {from=MaterialExpressionScalarParameter_6}
    B {from=MaterialExpressionConstant_7}
   ]
  }
  MaterialExpressionMultiply_6 {
   class=MaterialExpressionMultiply
   inputs [
    A {from=MaterialExpressionConstant_7}
    B {from=MaterialExpressionScalarParameter_8}
   ]
  }
  MaterialExpressionMultiply_7 {
   class=MaterialExpressionMultiply
   inputs [
    A {from=MaterialExpressionScalarParameter_7}
    B {from=MaterialExpressionConstant_7}
   ]
  }
  MaterialExpressionMultiply_8 {
   class=MaterialExpressionMultiply
   inputs [
    A {from=MaterialExpressionScalarParameter_9}
    B {from=MaterialExpressionConstant_7}
   ]
  }
  MaterialExpressionNamedRerouteDeclaration_0 {
   class=MaterialExpressionNamedRerouteDeclaration
   inputs [None {from=MaterialExpressionSaturate_5}]
   properties {name=SupeLevel_Mask, nodeColor=<rgba 0 0.0422 1 1>}
  }
  MaterialExpressionNamedRerouteDeclaration_1 {
   class=MaterialExpressionNamedRerouteDeclaration
   inputs [None {from=MaterialExpressionComponentMask_3}]
   properties {name=OriginalRoughness, nodeColor=<rgba 0.9732 1 0.9263 1>}
  }
  MaterialExpressionNamedRerouteDeclaration_2 {
   class=MaterialExpressionNamedRerouteDeclaration
   inputs [None {from=MaterialExpressionComponentMask_4}]
   properties {name=OriginalMetallic, nodeColor=<rgba 0.8047 0 1 1>}
  }
  MaterialExpressionNamedRerouteDeclaration_3 {
   class=MaterialExpressionNamedRerouteDeclaration
   inputs [None {from=MaterialExpressionComponentMask_2}]
   properties {name=OriginalSpecular, nodeColor=<rgba 0.2656 0 1 1>}
  }
  MaterialExpressionNamedRerouteDeclaration_4 {
   class=MaterialExpressionNamedRerouteDeclaration
   inputs [None {from=MaterialExpressionSubtract_5}]
   properties {name=SuperLevel_Mask_Emissive, nodeColor=<rgba 0 0.1562 1 1>}
  }
  MaterialExpressionNamedRerouteDeclaration_5 {
   class=MaterialExpressionNamedRerouteDeclaration
   inputs [None {from=MaterialExpressionGetMaterialAttributes_1, output=EmissiveColor}]
   properties {name=CurrentEmissive, nodeColor=<rgba 0.9922 1 0 1>}
  }
  MaterialExpressionNamedRerouteDeclaration_7 {
   class=MaterialExpressionNamedRerouteDeclaration
   inputs [None {from=MaterialExpressionSaturate_6}]
   properties {name=TimeAnimation, nodeColor=<rgba 0 0.25 1 1>}
  }
  MaterialExpressionNamedRerouteUsage_0 {class=MaterialExpressionNamedRerouteUsage}
  MaterialExpressionNamedRerouteUsage_1 {class=MaterialExpressionNamedRerouteUsage}
  MaterialExpressionNamedRerouteUsage_10 {class=MaterialExpressionNamedRerouteUsage}
  MaterialExpressionNamedRerouteUsage_2 {class=MaterialExpressionNamedRerouteUsage}
  MaterialExpressionNamedRerouteUsage_3 {class=MaterialExpressionNamedRerouteUsage}
  MaterialExpressionNamedRerouteUsage_4 {class=MaterialExpressionNamedRerouteUsage}
  MaterialExpressionNamedRerouteUsage_5 {class=MaterialExpressionNamedRerouteUsage}
  MaterialExpressionNamedRerouteUsage_6 {class=MaterialExpressionNamedRerouteUsage}
  MaterialExpressionNamedRerouteUsage_9 {class=MaterialExpressionNamedRerouteUsage}
  MaterialExpressionOneMinus_0 {class=MaterialExpressionOneMinus, inputs [None {from=MaterialExpressionSaturate_4}]}
  MaterialExpressionPreSkinnedNormal_1 {class=MaterialExpressionPreSkinnedNormal, inputs [], properties {}}
  MaterialExpressionPreSkinnedPosition_1 {class=MaterialExpressionPreSkinnedPosition, inputs [], properties {}}
  MaterialExpressionReroute_0 {class=MaterialExpressionReroute, inputs [None {from=MaterialExpressionFunctionInput_1}]}
  MaterialExpressionReroute_3 {class=MaterialExpressionReroute, inputs [None {from=MaterialExpressionSaturate_6}]}
  MaterialExpressionReroute_7 {class=MaterialExpressionReroute, inputs [None {from=MaterialExpressionReroute_0}]}
  MaterialExpressionReroute_8 {class=MaterialExpressionReroute, inputs [None {from=MaterialExpressionGetMaterialAttributes_0, output=MaterialAttributes}]}
  MaterialExpressionSaturate_0 {class=MaterialExpressionSaturate, inputs [None {from=MaterialExpressionAdd_0}]}
  MaterialExpressionSaturate_1 {class=MaterialExpressionSaturate, inputs [None {from=MaterialExpressionMultiply_1}]}
  MaterialExpressionSaturate_4 {class=MaterialExpressionSaturate, inputs [None {from=MaterialExpressionDivide_0}]}
  MaterialExpressionSaturate_5 {class=MaterialExpressionSaturate, inputs [None {from=MaterialExpressionAdd_2}]}
  MaterialExpressionSaturate_6 {class=MaterialExpressionSaturate, inputs [None {from=MaterialExpressionDivide_1}]}
  MaterialExpressionSaturate_9 {class=MaterialExpressionSaturate, inputs [None {from=MaterialExpressionSubtract_4}]}
  MaterialExpressionScalarParameter_0 {class=MaterialExpressionScalarParameter, properties {defaultValue=1, group=PiteousKicks_GoldMaterial, parameterName=RoughnessMultiplier}}
  MaterialExpressionScalarParameter_1 {class=MaterialExpressionScalarParameter, properties {group=PiteousKicks_GoldMaterial, parameterName=SpecularOffset}}
  MaterialExpressionScalarParameter_10 {class=MaterialExpressionScalarParameter, properties {defaultValue=5, group=PiteousKicks_GoldTransition, parameterName=GlowingEdgeThickness}}
  MaterialExpressionScalarParameter_11 {
   class=MaterialExpressionScalarParameter
   properties {defaultValue=2, group=PiteousKicks_GoldTransition, parameterName=TransitionDuration, sortPriority=50}
  }
  MaterialExpressionScalarParameter_12 {
   class=MaterialExpressionScalarParameter
   properties {defaultValue=2, group=PiteousKicks_GoldTransition, parameterName=TransitionDuration, sortPriority=50}
  }
  MaterialExpressionScalarParameter_13 {class=MaterialExpressionScalarParameter, properties {group=PiteousKicks_GoldTransition, parameterName=SL_CurrentLevel, sortPriority=5}}
  MaterialExpressionScalarParameter_14 {class=MaterialExpressionScalarParameter, properties {defaultValue=-9999, group=PiteousKicks_GoldTransition, parameterName=BP_GetTime}}
  MaterialExpressionScalarParameter_17 {class=MaterialExpressionScalarParameter, properties {group=PiteousKicks_GoldTransition, parameterName=SL_CurrentLevel, sortPriority=5}}
  MaterialExpressionScalarParameter_2 {class=MaterialExpressionScalarParameter, properties {group=PiteousKicks_GoldTransition, parameterName=SL_CurrentLevel, sortPriority=5}}
  MaterialExpressionScalarParameter_3 {class=MaterialExpressionScalarParameter, properties {defaultValue=-9999, group=PiteousKicks_GoldTransition, parameterName=BP_GetTime}}
  MaterialExpressionScalarParameter_6 {
   class=MaterialExpressionScalarParameter
   properties {defaultValue=-80, group=PiteousKicks_GoldTransition, parameterName=MaskPos_01, sortPriority=6}
  }
  MaterialExpressionScalarParameter_7 {
   class=MaterialExpressionScalarParameter
   properties {defaultValue=-10, group=PiteousKicks_GoldTransition, parameterName=MaskPos_02, sortPriority=6}
  }
  MaterialExpressionScalarParameter_8 {
   class=MaterialExpressionScalarParameter
   properties {defaultValue=50, group=PiteousKicks_GoldTransition, parameterName=MaskPos_03, sortPriority=6}
  }
  MaterialExpressionScalarParameter_9 {
   class=MaterialExpressionScalarParameter
   properties {defaultValue=100, group=PiteousKicks_GoldTransition, parameterName=MaskPos_04, sortPriority=6}
  }
  MaterialExpressionSetMaterialAttributes_0 {
   class=MaterialExpressionSetMaterialAttributes
   inputs [
    MaterialAttributes {from=MaterialExpressionReroute_0, output=""}
    BaseColor {from=MaterialExpressionTextureSampleParameter2D_1, output=RGB}
    Specular {from=MaterialExpressionTextureSampleParameter2D_2, output=R}
    Roughness {from=MaterialExpressionTextureSampleParameter2D_2, output=R}
    Metallic {from=MaterialExpressionTextureSampleParameter2D_2, output=R}
    EmissiveColor {from=MaterialExpressionConstant_4, output=""}
    Normal {from=MaterialExpressionFunctionInput_2, output=""}
    Opacity {from=MaterialExpressionConstant_5, output=""}
   ]
   properties {attributeSetTypes [
     #5
     #1
     #2
     #3
     #4
     #6
     #7
    ]}
  }
  MaterialExpressionSetMaterialAttributes_1 {
   class=MaterialExpressionSetMaterialAttributes
   inputs [
    MaterialAttributes {from=MaterialExpressionGetMaterialAttributes_1, output=MaterialAttributes}
    EmissiveColor {from=MaterialExpressionLinearInterpolate_2, output=""}
   ]
   properties {attributeSetTypes [#4]}
  }
  MaterialExpressionShaderStageSwitch_2 {
   class=MaterialExpressionShaderStageSwitch
   inputs [
    PixelShader {from=MaterialExpressionVertexInterpolator_3, output=PS}
    VertexShader {from=MaterialExpressionPreSkinnedNormal_1, output=""}
   ]
   properties {}
  }
  MaterialExpressionShaderStageSwitch_3 {
   class=MaterialExpressionShaderStageSwitch
   inputs [
    PixelShader {from=MaterialExpressionVertexInterpolator_0, output=PS}
    VertexShader {from=MaterialExpressionPreSkinnedPosition_1, output=""}
   ]
   properties {}
  }
  MaterialExpressionSmoothStep_0 {
   class=MaterialExpressionSmoothStep
   inputs [
    Min {from="", output=""}
    Max {from="", output=""}
    Value {from=MaterialExpressionOneMinus_0, output=""}
   ]
   properties {constMax=0.25}
  }
  MaterialExpressionStaticSwitchParameter_0 {
   class=MaterialExpressionStaticSwitchParameter
   inputs [
    True {from=MaterialExpressionTextureSampleParameter2D_0, output=RGB}
    False {from=MaterialExpressionMaterialFunctionCall_0, output=RGB}
   ]
   properties {group=PiteousKicks_GoldMaterial, parameterName=OverrideBaseSRM, sortPriority=200}
  }
  MaterialExpressionSubtract_0 {
   class=MaterialExpressionSubtract
   inputs [
    A {from=MaterialExpressionTime_0}
    B {from=MaterialExpressionScalarParameter_14}
   ]
  }
  MaterialExpressionSubtract_1 {
   class=MaterialExpressionSubtract
   inputs [
    A {from=MaterialExpressionComponentMask_5}
    B {from=MaterialExpressionMultiply_4}
   ]
  }
  MaterialExpressionSubtract_2 {
   class=MaterialExpressionSubtract
   inputs [
    A {from=MaterialExpressionScalarParameter_2}
    B {from=""}
   ]
  }
  MaterialExpressionSubtract_3 {
   class=MaterialExpressionSubtract
   inputs [
    A {from=MaterialExpressionAdd_5}
    B {from=MaterialExpressionTime_1}
   ]
  }
  MaterialExpressionSubtract_4 {
   class=MaterialExpressionSubtract
   inputs [
    A {from=MaterialExpressionAdd_2}
    B {from=MaterialExpressionScalarParameter_10}
   ]
   properties {constB=10}
  }
  MaterialExpressionSubtract_5 {
   class=MaterialExpressionSubtract
   inputs [
    A {from=MaterialExpressionSaturate_5}
    B {from=MaterialExpressionSaturate_9}
   ]
  }
  MaterialExpressionSwitch_0 {
   class=MaterialExpressionSwitch
   inputs [
    SwitchValue {from=MaterialExpressionScalarParameter_2, output=""}
    Default {from=MaterialExpressionMultiply_5, output=""}
    None {from=MaterialExpressionMultiply_5, output=""}
    None {from=MaterialExpressionMultiply_7, output=""}
    None {from=MaterialExpressionMultiply_6, output=""}
    None {from=MaterialExpressionMultiply_8, output=""}
   ]
   properties {inputs [
     {
      inputName=None
      input {
       expression=&self:MaterialExpressionMultiply_5
       inputName=None
       mask=0
       maskA=0
       maskB=0
       maskG=0
       maskR=0
       outputIndex=0
      }
     }
     {
      inputName=None
      input {
       expression=&self:MaterialExpressionMultiply_7
       inputName=None
       mask=0
       maskA=0
       maskB=0
       maskG=0
       maskR=0
       outputIndex=0
      }
     }
     {
      inputName=None
      input {
       expression=&self:MaterialExpressionMultiply_6
       inputName=None
       mask=0
       maskA=0
       maskB=0
       maskG=0
       maskR=0
       outputIndex=0
      }
     }
     {
      inputName=None
      input {
       expression=&self:MaterialExpressionMultiply_8
       inputName=None
       mask=0
       maskA=0
       maskB=0
       maskG=0
       maskR=0
       outputIndex=0
      }
     }
    ]}
  }
  MaterialExpressionSwitch_1 {
   class=MaterialExpressionSwitch
   inputs [
    SwitchValue {from=MaterialExpressionSubtract_2, output=""}
    Default {from=MaterialExpressionMultiply_5, output=""}
    None {from=MaterialExpressionMultiply_5, output=""}
    None {from=MaterialExpressionMultiply_7, output=""}
    None {from=MaterialExpressionMultiply_6, output=""}
    None {from=MaterialExpressionMultiply_8, output=""}
   ]
   properties {inputs [
     {
      inputName=None
      input {
       expression=&self:MaterialExpressionMultiply_5
       inputName=None
       mask=0
       maskA=0
       maskB=0
       maskG=0
       maskR=0
       outputIndex=0
      }
     }
     {
      inputName=None
      input {
       expression=&self:MaterialExpressionMultiply_7
       inputName=None
       mask=0
       maskA=0
       maskB=0
       maskG=0
       maskR=0
       outputIndex=0
      }
     }
     {
      inputName=None
      input {
       expression=&self:MaterialExpressionMultiply_6
       inputName=None
       mask=0
       maskA=0
       maskB=0
       maskG=0
       maskR=0
       outputIndex=0
      }
     }
     {
      inputName=None
      input {
       expression=&self:MaterialExpressionMultiply_8
       inputName=None
       mask=0
       maskA=0
       maskB=0
       maskG=0
       maskR=0
       outputIndex=0
      }
     }
    ]}
  }
  MaterialExpressionTextureObjectParameter_0 {
   class=MaterialExpressionTextureObjectParameter
   inputs [
    UVs {from="", output=""}
    Apply View MipBias {from="", output=""}
   ]
   properties {
    group=SuperLevelling_Mask
    parameterName=SL_TransitionNoise
    texture=/Game/Characters/Player/Male/Medium/Bodies/M_MED_Viper/Textures/Glitch/T_SoftNoise
   }
  }
  MaterialExpressionTextureSampleParameter2D_0 {
   class=MaterialExpressionTextureSampleParameter2D
   inputs [
    UVs {from="", output=""}
    Apply View MipBias {from="", output=""}
   ]
   properties {
    group=PiteousKicks_GoldMaterial
    parameterName=Gold_SRM
    samplerSource=SSM_Wrap_WorldGroupSettings
    samplerType=SAMPLERTYPE_Masks
    sortPriority=201
    texture=/Game/Characters/Player/Male/Medium/Bodies/M_MED_PiteousKicks/Textures/T_PiteousKicks_Body_S
   }
  }
  MaterialExpressionTextureSampleParameter2D_1 {
   class=MaterialExpressionTextureSampleParameter2D
   inputs [
    UVs {from="", output=""}
    Apply View MipBias {from="", output=""}
   ]
   properties {
    group=PiteousKicks_GoldMaterial
    parameterName=GoldTexture_Diffuse
    samplerSource=SSM_Wrap_WorldGroupSettings
    texture=/Game/Characters/Player/Male/Medium/Bodies/M_MED_PiteousKicks_Pine/Skins/Gleam/Textures/T_PiteousKicks_PineGleam_Body_D
   }
  }
  MaterialExpressionTextureSampleParameter2D_2 {
   class=MaterialExpressionTextureSampleParameter2D
   inputs [
    UVs {from="", output=""}
    Apply View MipBias {from="", output=""}
   ]
   properties {
    group=PiteousKicks_GoldMaterial
    parameterName=GoldTexture_SRM
    samplerSource=SSM_Wrap_WorldGroupSettings
    samplerType=SAMPLERTYPE_Masks
    texture=/Game/Characters/Player/Male/Medium/Bodies/M_MED_PiteousKicks_Pine/Skins/Gleam/Textures/T_PiteousKicks_PineGleam_Body_S
   }
  }
  MaterialExpressionTime_0 {class=MaterialExpressionTime}
  MaterialExpressionTime_1 {class=MaterialExpressionTime}
  MaterialExpressionVectorParameter_0 {
   class=MaterialExpressionVectorParameter
   properties {group=PiteousKicks_GoldMaterial, parameterName=GoldColor, sortPriority=30, defaultValue=<rgba 0.7084 0.4233 0.1022 1>}
  }
  MaterialExpressionVectorParameter_1 {
   class=MaterialExpressionVectorParameter
   properties {group=PiteousKicks_GoldMaterial, parameterName=Transition_EmissiveColor, sortPriority=101, defaultValue=<rgba 2 0 1.4795 1>}
  }
  MaterialExpressionVectorParameter_3 {
   class=MaterialExpressionVectorParameter
   properties {group=PiteousKicks_GoldTransition, parameterName=GoldenNoiseTiling_02, defaultValue=<rgba 6 4 80 1>}
  }
  MaterialExpressionVectorParameter_4 {
   class=MaterialExpressionVectorParameter
   properties {group=PiteousKicks_GoldTransition, parameterName=GoldenNoiseTiling_01, defaultValue=<rgba 6 4 80 1>}
  }
  MaterialExpressionVertexInterpolator_0 {class=MaterialExpressionVertexInterpolator, inputs [VS {from=MaterialExpressionPreSkinnedPosition_1, output=""}], properties {}}
  MaterialExpressionVertexInterpolator_3 {class=MaterialExpressionVertexInterpolator, inputs [VS {from=MaterialExpressionPreSkinnedNormal_1, output=""}], properties {}}
 }}
