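# Material function for the PiteousKicks Pine pickaxe that adds a ghostly animated glow and rippling normals to incoming material attributes.
# Emissive adds a tinted, panning swirl texture, offset by a rerouted value and masked by the Fresnel input.
# Normals add triplanar-style scrolling noise in pre-skinned space, minus a position-derived axis normal, converted local to tangent.
# A second Fresnel masks that distortion; a second output is named OriginalNormal.
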MaterialFunction /BRCosmetics/Weapons/FORT_Melee/Pickaxe_PiteousKicks/Skins/Pine/Materials/MF_PiteousKicks_Pine_Pickaxe_GhostlyGlow
graph {nodes [
  FunctionOutput(None=NamedRerouteUsage()) {outputName=OriginalNormal, sortPriority=1}
  n1 = GetMaterialAttributes(None=FunctionInput() {inputName=MaterialAttributes, inputType=FunctionInput_MaterialAttributes}) {attributeGetTypes [#1, #2]}
  n2 = Power(
   Base=FunctionInput() {inputName=Fresnel, inputType=FunctionInput_Scalar, sortPriority=10}
   Exp=ScalarParameter() {defaultValue=2, group=PiteousKicks_EmissiveLook, parameterName=FrenelPower, sortPriority=10}
  )
  n3 = VectorParameter() {group=PiteousKicks_EmissiveLook, parameterName=NoiseSpeed, sortPriority=50, defaultValue=<rgba 0 0.02 0 0>}
  n4 = Panner(Coordinate=TextureCoordinate(), Speed=ComponentMask(None=@n3.RGB) {g=true, r=true}) {speedX=0.1}
  n5 = ScalarParameter() {defaultValue=2, group=PiteousKicks_EmissiveLook, parameterName=NoiseTiling}
  n6 = ComponentMask(None=NamedRerouteUsage()) {g=true, r=true}
  n7 = Constant() {r=0.2}
  n8 = TextureSample(UVs=Add(A=Multiply(A=@n4, B=@n5) {constB=2}, B=Multiply(A=@n6, B=@n7))) {
   samplerSource=SSM_Wrap_WorldGroupSettings
   samplerType=SAMPLERTYPE_LinearColor
   texture=/Game/Global/Materials/MaterialGym/Materials/GhostNoise/Textures/T_ghost_swirls_01
  }
  n9 = VectorParameter() {group=PiteousKicks_EmissiveLook, parameterName=GlowColor, sortPriority=10, defaultValue=<rgba 0 1 0.1261 1>}
  n10 = Multiply(
   A=Multiply(
    A=Multiply(A=@n2, B=@n8.R) {constB=1.3}
    B=ScalarParameter() {defaultValue=2, group=PiteousKicks_EmissiveLook, parameterName=GlowIntensity, sortPriority=10}
   )
   B=@n9.RGB
  )
  FunctionOutput(None=SetMaterialAttributes!(MaterialAttributes=@n1.MaterialAttributes, EmissiveColor=Add(A=@n1.EmissiveColor, B=@n10), Normal=NamedRerouteUsage()) {attributeSetTypes [#1, #2]})
  n11 = VertexInterpolator(VS=PreSkinnedPosition())
  n12 = Time()
  n13 = VectorParameter() {group=PiteousKicks_EmissiveLook, parameterName=NormalNoise_Speed, sortPriority=66, defaultValue=<rgba 0 1 -10 1>}
  n14 = Add(
   A=@n11.PS
   B=Multiply(A=@n12, B=MaterialFunctionCall(B=@n13.B) {materialFunction=&AppendMany}.RGB)
  )
  n15 = ScalarParameter() {defaultValue=0.04, group=PiteousKicks_EmissiveLook, parameterName=NormalNoise_Tiling, sortPriority=66}
  n16 = TextureSample(UVs=ComponentMask(None=Multiply(A=@n14, B=@n15) {constB=0.1}) {b=true, g=true}) {
   samplerSource=SSM_Wrap_WorldGroupSettings
   samplerType=SAMPLERTYPE_Normal
   texture=/Game/Characters/Player/Male/Medium/Bodies/M_MED_Cat_Burglar_Ghost/Textures/T_M_MED_Cat_Burglar_Ghost_N1
  }
  n17 = Add(
   A=@n11.PS
   B=Multiply(A=@n12, B=MaterialFunctionCall(R=@n13.R, B=@n13.R) {materialFunction=&AppendMany}.RGB)
  )
  n18 = TextureSample(UVs=ComponentMask(None=Multiply(A=@n17, B=@n15) {constB=0.1}) {b=true, r=true}) {
   samplerSource=SSM_Wrap_WorldGroupSettings
   samplerType=SAMPLERTYPE_Normal
   texture=&T_M_MED_Cat_Burglar_Ghost_N1
  }
  n19 = VertexInterpolator(VS=PreSkinnedNormal())
  n20 = Multiply(A=@n19.PS) {constB=-1}
  n21 = Add(A=Saturate(None=ComponentMask(None=@n19.PS) {g=true}), B=Saturate(None=ComponentMask(None=@n20) {g=true}))
  n22 = Add(
   A=@n11.PS
   B=Multiply(A=@n12, B=MaterialFunctionCall(R=@n13.R) {materialFunction=&AppendMany}.RGB)
  )
  n23 = TextureSample(UVs=ComponentMask(None=Multiply(A=@n22, B=@n15)) {g=true, r=true}) {
   samplerSource=SSM_Wrap_WorldGroupSettings
   samplerType=SAMPLERTYPE_Normal
   texture=&T_M_MED_Cat_Burglar_Ghost_N1
  }
  n24 = Saturate(None=ComponentMask(None=@n19.PS) {b=true})
  n25 = Saturate(None=ComponentMask(None=Multiply(A=@n19.PS) {constB=-1}) {b=true})
  n26 = LinearInterpolate(
   A=LinearInterpolate(A=@n16.RGB, B=@n18.RGB, Alpha=SmoothStep(Value=@n21) {constMin=0.4})
   B=@n23.RGB
   Alpha=SmoothStep(Value=Add(A=@n24, B=@n25)) {constMin=0.4}
  )
  n27 = VectorParameter() {group=PiteousKicks_EmissiveLook, parameterName=NormalNoise_ActorOffset, sortPriority=67, defaultValue=<rgba 0 0 0 1>}
  n28 = Add(A=@n11.PS, B=@n27.RGB)
  n29 = Multiply(A=Constant3Vector() {constant=<rgba 0 1 0 0>}, B=ComponentMask(None=@n28) {g=true})
  n30 = ComponentMask(None=@n28) {r=true}
  n31 = Multiply(A=Constant3Vector() {constant=<rgba 1 0 0 0>}, B=@n30)
  n32 = Constant() {r=10}
  n33 = MaterialFunctionCall(In=Abs(None=@n30), Contrast=@n32) {materialFunction=&CheapContrast}
  n34 = Constant3Vector() {constant=<rgba 0 0 1 0>}
  n35 = ComponentMask(None=@n28) {b=true}
  n36 = Abs(None=@n35)
  n37 = Normalize(VectorInput=LinearInterpolate(
    A=LinearInterpolate(A=@n29, B=@n31, Alpha=@n33.Result)
    B=Multiply(A=@n34, B=@n35)
    Alpha=MaterialFunctionCall(In=@n36, Contrast=@n32) {materialFunction=&CheapContrast}.Result
   ))
  n38 = ScalarParameter() {defaultValue=2, group=PiteousKicks_EmissiveLook, parameterName=NormalNoise_FresnelMaskingContrast, sortPriority=66}
  n39 = Add(
   A=NamedRerouteUsage()
   B=ScalarParameter() {group=PiteousKicks_EmissiveLook, parameterName=NormalNoise_FresnelMasking_Z_Offset, sortPriority=66}
  )
  n40 = Fresnel(
   ExponentIn=LinearInterpolate(
    A=ScalarParameter() {defaultValue=2, group=PiteousKicks_EmissiveLook, parameterName=NormalNoise_FresnelMaskingPower_01, sortPriority=66}
    B=ScalarParameter() {defaultValue=2, group=PiteousKicks_EmissiveLook, parameterName=NormalNoise_FresnelMaskingPower_02, sortPriority=66}
    Alpha=Saturate(None=@n39)
   )
   Normal=VertexNormalWS()
  )
  n41 = Multiply(
   A=Multiply(
    A=ScalarParameter() {defaultValue=1, group=PiteousKicks_EmissiveLook, parameterName=NormalNoise_Intensity, sortPriority=66}
    B=Subtract(A=@n26, B=@n37)
   )
   B=Saturate(None=Multiply(A=@n38, B=@n40))
  )
  NamedRerouteDeclaration(None=Add(A=NamedRerouteUsage(), B=Transform(None=@n41) {transformSourceType=TRANSFORMSOURCE_Local, transformType=TRANSFORM_Tangent})) {name=Normal, nodeColor=<rgba 1 0.0938 0 1>}
  NamedRerouteDeclaration(None=@n1.Normal) {name=OriginalNormal, nodeColor=<rgba 1 0 0.4219 1>}
  NamedRerouteDeclaration(None=ComponentMask(None=@n11.PS) {b=true}) {name=Preskinned_Z_Position, nodeColor=<rgba 0.0312 1 0 1>}
  ScalarParameter() {defaultValue=-10, group=PiteousKicks_EmissiveLook, parameterName=NormalNoise_Speed, sortPriority=66}
  Transform(None=VertexNormalWS()) {transformSourceType=TRANSFORMSOURCE_World, transformType=TRANSFORM_Local}
 ]}
