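# Parent material for the Pine skin of the PiteousKicks pickaxe, built on the 2023 character parent material functions with material attributes.
# The PreFX function feeds a skin-specific GhostlyGlow function, whose result goes into PostFX along with mask, position, distance blend and fresnel inputs.
# A PickaxeTransition function labeled Legacy also branches from the glow result.
# Shading model comes from the material expression, spanning lit, subsurface, clear coat and cloth types.
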Material /BRCosmetics/Weapons/FORT_Melee/Pickaxe_PiteousKicks/Skins/Pine/Materials/M_Pickaxe_PiteousKicks_Pine_PARENT
graph {
 properties {
  bUseMaterialAttributes=true
  bUsedWithInstancedStaticMeshes=true
  bUsedWithMeshDeformer=true
  bUsedWithSkeletalMesh=true
  pixelDepthOffsetMode=PDOM_Legacy
  shadingModel=MSM_FromMaterialExpression
  usedShadingModels=Default Lit | Subsurface | Clear Coat | Subsurface Profile | Cloth
 }
 nodes [
  n1 = MaterialFunctionCall() {materialFunction=&MF_CharacterParent_2023_PreFX}
  n2 = MaterialFunctionCall!(MaterialAttributes=@n1.Result, Fresnel=@n1.Result) {materialFunction=&/MF_PiteousKicks_Pine_Pickaxe_GhostlyGlow}
  MaterialFunctionCall(Material=@n2.Result, M=@n1.M, Pre Skinned Local Position=@n1.M, Distance Blend=@n1.M, Fresnel=@n1.M) {materialFunction=&MF_CharacterParent_2023_PostFX}
  MaterialFunctionCall!(MaterialAttributes=@n2.Result, OriginalNormal=@n2.Result) {desc=Legacy, materialFunction=&/MF_PiteousKicks_PickaxeTransition}
 ]
}
